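# Niagara effect for the RemoteControl emote: a short gust sweeping along X, with air rings, a ribbon trail, leaves, dust and wind streaks.
# A zero-size Source particle follows a curved dip path and spawns the oriented Main_Ring meshes; Air_Trail and both leaf emitters repeat that path formula.
# Dust, Puffs and Swoosh sweep along the base, using two spawn groups to spray backward to either side.
# Several colors adapt to time of day.
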NiagaraSystem /BRCosmetics/Effects/Fort_Effects/Emotes/RemoteControl/NS_Emote_RemoteControl
dependencies {
 usedDataInterfaces [
  &NiagaraDataInterfaceParticleRead
  &NiagaraDataInterfaceCurve
  &NiagaraDataInterfaceColorCurve
  &NiagaraDataInterfacePlatformSet
  &NiagaraDataInterfaceVectorCurve
  &NiagaraDataInterfaceMeshRendererInfo
 ]
 usedDynamicInputs [&SimulationPosition, /Niagara/DynamicInputs/CascadeConversion/SplitVector]
 usedModules [&SystemState]
}
emitters {
 Air_Trail {
  emitterUpdateScript [
   &EmitterState()
   &SpawnBurst_Instantaneous(Spawn Count=20, Use Spawn Probability=false, UseLoopCountLimit=true)
  ]
  particleSpawnScript [
   &InitializeParticle(
    Color=&MultiplyLinearColor(
     Linear Color=&FNTimeOfDayColor(
      Night Color=<rgba 0.5156 0.6758 0.75 0.5>
      Morning Color=<rgba 0.1934 0.2534 0.2812 0.6>
      Day Color=<rgba 0.6875 0.901 1 0.7>
      Evening Color=<rgba 0.3044 0.3989 0.4427 0.6>
     )
     Scale Factor=&MakeFloat4FromVectorAndFloat(
      "Vector (RGB)"=<1 1 1>
      "Float (Alpha)"=&FloatFromCurve(
       FloatCurve=DataInterface() {propertyValues {Curve {keys [
           {arriveTangent=5.3189, interpMode=RCIM_Cubic, leaveTangent=5.3189, tangentMode=RCTM_User, time=0, value=0}
           {arriveTangent=2, leaveTangent=2, time=0.1, value=1}
           <at 0.5 1>
           {arriveTangent=-2, interpMode=RCIM_Cubic, leaveTangent=-2, time=0.9, value=1}
           {arriveTangent=-5.94, tangentMode=RCTM_User, time=1, value=0}
          ]}}}
       CurveIndex=&ReturnNormalizedExecIndex()
      )
     )
    )
    Mass Mode=::NewEnumerator1
    Ribbon Width=&FloatFromCurve(
     FloatCurve=DataInterface() {propertyValues {Curve {keys [<at 0 0.2>,<at 1 1.2>]}}}
     CurveIndex=&ReturnNormalizedExecIndex()
     Scale Curve=120
    )
    Ribbon Width Mode=::NewEnumerator1
    UsePositionOffset=true
    Write Material Random=true
   )
   /SetVariables(Particles.Position=&AddVectorToPosition(
     Position=<0 0 0>
     Vector=&MakeVector(
      X=&Lerp_Float(A=-300, B=200, Alpha=&ReturnNormalizedExecIndex())
      Y=0
      Z=&Add_Float(
       A=&FloatFromCurve(
        FloatCurve=DataInterface() {propertyValues {Curve {keys [
            {arriveTangent=-2.1826, interpMode=RCIM_Cubic, leaveTangent=-2.1826, tangentMode=RCTM_User, time=0, value=1.5}
            {arriveTangent=-0.5, interpMode=RCIM_Cubic, leaveTangent=-0.5, time=0.5, value=0.5}
            {arriveTangent=2.8781, interpMode=RCIM_Cubic, leaveTangent=2.8781, tangentMode=RCTM_User, time=1, value=1}
           ]}}}
        CurveIndex=&ReturnNormalizedExecIndex()
        Scale Curve=50
       )
       B=50
      )
     )
    )) {isSetParameters=true}
   /SetVariables(Particles.RibbonLinkOrder:NiagaraFloat=&ReturnNormalizedExecIndex()) {isSetParameters=true}
  ]
  particleUpdateScript [
   &ParticleState()
   &ScaleSpriteSize(
    Scale Sprite Size Mode=::NewEnumerator0
    Uniform Scale Factor=&Multiply_Float(
     A=&FloatFromCurve(
      FloatCurve=DataInterface() {propertyValues {Curve {keys [
          {interpMode=RCIM_Cubic, tangentMode=RCTM_User, time=0, value=0}
          {interpMode=RCIM_Cubic, tangentMode=RCTM_User, time=0.5, value=1}
          {tangentMode=RCTM_User, time=1, value=0}
         ]}}}
      CurveIndex=&ClampFloat(Float=&Add_Float(
        A=&Multiply_Float(A=&ReturnNormalizedExecIndex(), B=2)
        B=&FloatFromCurve(FloatCurve=DataInterface() {propertyValues {Curve {keys [{tangentMode=RCTM_User, time=0, value=1}, {tangentMode=RCTM_User, time=1, value=-2}]}}})
       ))
     )
     B=1
    )
   )
   &ScaleColor(
    Scale Alpha=&FloatFromCurve(
     FloatCurve=DataInterface() {propertyValues {Curve {keys [
         {interpMode=RCIM_Cubic, tangentMode=RCTM_User, time=0, value=0}
         {arriveTangent=0.2265, leaveTangent=2, tangentMode=RCTM_User, time=0.3, value=1}
         <at 0.5 1>
         {arriveTangent=0.0955, interpMode=RCIM_Cubic, leaveTangent=0.0955, tangentMode=RCTM_User, time=0.7, value=1}
         {tangentMode=RCTM_User, time=1, value=0}
        ]}}}
     CurveIndex=&ClampFloat(Float=&Add_Float(
       A=&Multiply_Float(A=&ReturnNormalizedExecIndex(), B=2)
       B=&FloatFromCurve(FloatCurve=DataInterface() {propertyValues {Curve {keys [{arriveTangent=0.9892, interpMode=RCIM_Cubic, time=0, value=1}, {interpMode=RCIM_Cubic, leaveTangent=0.2296, time=1, value=-2.5}]}}})
      ))
     Scale Curve=0.5
    )
    ScaleA=true
    ScaleRGB=true
   )
   &ScaleRibbonWidth()
  ]
  rendererClasses [&NiagaraRibbonRendererProperties]
  renderers [{rendererClass=&NiagaraRibbonRendererProperties}]
 }
 Dust {
  emitterUpdateScript [
   &EmitterState(Life Cycle Mode=::NewEnumerator1, Loop Behavior=::NewEnumerator1, Loop Delay=0.2, Loop Duration=0.4, UseLoopDelay=true)
   &SpawnRate(SpawnRate=20, Use Spawn Probability=false)
   &SpawnRate(SpawnGroup=1, SpawnRate=20, Use Spawn Probability=false) {module=SpawnRate001}
  ]
  particleSpawnScript [
   &InitializeParticle(
    Color=<rgba 0.349 0.2519 0.1708 0.8>
    Lifetime Max=0.8
    Lifetime Min=0.7
    Lifetime Mode=::NewEnumerator1
    Sprite Rotation Mode=::NewEnumerator2
    Sprite Size Mode=::NewEnumerator3
    Sprite UV Mode=::NewEnumerator3
    Uniform Sprite Size=&Multiply_Float(
     A=&RandomRangeFloat(Minimum=50, Maximum=60)
     B=&FloatFromCurve(FloatCurve=DataInterface() {propertyValues {Curve {keys [<at 0 1>,<at 1 0.7>]}}}, CurveIndex=$Emitter.NormalizedLoopAge)
    )
    UsePositionOffset=false
    Write Material Random=true
   )
   &SystemLocation(
    MaskSpawnGroup=false
    Offset=&MakeVector(X=&Lerp_Float(A=-90, B=125, Alpha=$Emitter.NormalizedLoopAge), Y=0, Z=4)
   )
   &AddVelocity(
    Velocity=&Lerp_Vector(
     A=<-200 135 50>
     B=<-200 -135 50>
     Alpha=&MakeFloatFromInt(Int=&SpawnGroup())
    )
    Velocity Speed Scale=&FloatFromCurve(
     FloatCurve=DataInterface() {propertyValues {Curve {keys [
         {interpMode=RCIM_Cubic, tangentMode=RCTM_User, time=0, value=1}
         {arriveTangent=-0.5471, interpMode=RCIM_Cubic, leaveTangent=-0.5471, tangentMode=RCTM_User, time=1, value=0.8}
        ]}}}
     CurveIndex=$Emitter.NormalizedLoopAge
    )
   )
  ]
  particleUpdateScript [
   &ParticleState()
   &Drag(Drag=2, Use Linear Drag=true, Use Rotational Drag=true)
   &SolveForcesAndVelocity(Clamp Velocity=false, Limit Acceleration=false)
   &DynamicMaterialParameters(
    Index 0 Param 1=&FloatFromCurve(
     FloatCurve=DataInterface() {propertyValues {Curve {keys [
         {arriveTangent=2.7533, interpMode=RCIM_Cubic, time=0, value=0}
         {arriveTangent=0.0529, interpMode=RCIM_Cubic, leaveTangent=0.2021, tangentMode=RCTM_User, time=1, value=1}
        ]}}}
     Scale Curve=0.1
    )
    Index 0 Param 4=0.5
    Index 2 Param 1=&FloatFromCurve(FloatCurve=DataInterface() {propertyValues {Curve {keys [{tangentMode=RCTM_User, time=0, value=0}, {tangentMode=RCTM_User, time=1, value=1}]}}})
    Index 2 Param 2=&FloatFromCurve(
     FloatCurve=DataInterface() {propertyValues {Curve {keys [{tangentMode=RCTM_User, time=0, value=0}, {tangentMode=RCTM_User, time=1, value=1}]}}}
     Scale Curve=&RandomRangeFloat(Minimum=0.5, Maximum=1)
    )
    Index 2 Param 3=2
    Index 2 Param 4=1
    Module.Index0Param0Write=true
    Module.Index0Param1Write=true
    Module.Index0Param2Write=true
    Module.Index0Param3Write=true
    Module.Index1Param0Write=true
    Module.Index1Param1Write=true
    Module.Index1Param2Write=true
    Module.Index1Param3Write=true
    Module.Index2Param0Write=true
    Module.Index2Param1Write=true
    Module.Index2Param2Write=true
    Module.Index2Param3Write=true
    Write Parameter Index 1=true
    Write Parameter Index 2=true
   )
   &ScaleColor(
    Scale Alpha=&FloatFromCurve(FloatCurve=DataInterface() {propertyValues {Curve {keys [
         {arriveTangent=9.0725, interpMode=RCIM_Cubic, leaveTangent=9.0725, tangentMode=RCTM_User, time=0, value=0}
         {arriveTangent=-2.5249, leaveTangent=-2.5249, time=0.1408, value=1}
         {arriveTangent=-0.0505, interpMode=RCIM_Cubic, leaveTangent=-0.0505, tangentMode=RCTM_User, time=0.3, value=1}
         {arriveTangent=-2.2624, interpMode=RCIM_Cubic, leaveTangent=-2.2624, tangentMode=RCTM_User, time=1, value=0}
        ]}}})
    ScaleA=true
    ScaleRGB=true
   )
   &ScaleSpriteSize(Uniform Curve Sprite Scale=DataInterface() {propertyValues {Curve {keys [
        {arriveTangent=1.0717, interpMode=RCIM_Cubic, time=0, value=0.5}
        {arriveTangent=1.1442, interpMode=RCIM_Cubic, leaveTangent=1.1442, tangentMode=RCTM_User, time=1, value=1.2}
       ]}}})
  ]
  rendererClasses [&NiagaraSpriteRendererProperties]
  renderers [{rendererClass=&NiagaraSpriteRendererProperties}]
 }
 Leaves_Drag {
  emitterSpawnScript [
   /Game/Effects/Niagara/MasterMaterials/ToonShader/Template/NM_ToonShaderParameters(
    BaseColor:LinearColor=<rgba 0.1117 0.3646 0 1>
    HighLight_Size=0.15
    HighLight_Softness=1
    InkLine_Color:LinearColor=<rgba 0.2025 0.3958 0.1533 1>
    Metallic_Reflection_Color:LinearColor=<rgba 0.0097 0.5573 0 1>
    RimLight_Color:LinearColor=<rgba 0 0.9427 0.0941 1>
   )
   /Game/Effects/Niagara/MasterMaterials/ToonShader/Template/NM_OutlineParameters(OutlineColor:LinearColor=<rgba 0.0162 0.0417 0.0042 1>, OutlineWidth=0.75)
  ]
  emitterUpdateScript [
   &EmitterState(Life Cycle Mode=::NewEnumerator1, Loop Behavior=::NewEnumerator1, Loop Delay=0.35, Loop Duration=0.2, UseLoopDelay=true)
   &SpawnRate(SpawnRate=30, Use Spawn Probability=false)
   &SpawnBurst_Instantaneous(
    Spawn Count=&RandomRangeInt(Minimum=5, Maximum=7)
    Spawn Time=0.3
    Use Spawn Probability=false
    UseLoopCountLimit=true
   ) {enabled=false}
  ]
  particleSpawnScript [
   &InitializeParticle(
    Color=&FNTimeOfDayColor(
     Night Color=<rgba 0.117 0.2664 0.1441 1>
     Morning Color=<rgba 0.3729 0.5052 0.1395 1>
     Day Color=<rgba 0.3215 0.4271 0.1245 1>
     Evening Color=<rgba 0.3646 0.4844 0.1412 1>
    )
    Lifetime Max=1.5
    Lifetime Min=0.8
    Lifetime Mode=::NewEnumerator1
    Mesh Scale Mode=::NewEnumerator1
    Mesh Uniform Scale Max=0.9
    Mesh Uniform Scale Min=0.8
    Sprite Size Mode=::NewEnumerator3
    Uniform Sprite Size=20
    UsePositionOffset=true
    Write Material Random=true
   )
   &InitialMeshOrientation(Use Rotation Vector=false)
   &ShapeLocation(
    Disc Coverage=1
    MaskSpawnGroup=false
    Override Seed=false
    Ring Radius=20
    Rotation Mode=::NewEnumerator1
    Shape Primitive=::NewEnumerator4
    UseOwnerScale=true
   )
   /SetVariables(Particles.Position=&AddVectorToPosition(
     Position=$Particles.Initial.Position
     Vector=&MakeVector(
      X=&Lerp_Float(A=-225, B=200, Alpha=&RandomRangeFloat(Minimum=0.3, Maximum=0.6))
      Y=0
      Z=&Add_Float(
       A=&FloatFromCurve(
        FloatCurve=DataInterface() {propertyValues {Curve {keys [
            {arriveTangent=-2.1826, interpMode=RCIM_Cubic, leaveTangent=-2.1826, tangentMode=RCTM_User, time=0, value=1.5}
            {arriveTangent=-0.5, interpMode=RCIM_Cubic, leaveTangent=-0.5, time=0.5, value=0.5}
            {arriveTangent=2.8781, interpMode=RCIM_Cubic, leaveTangent=2.8781, tangentMode=RCTM_User, time=1, value=1}
           ]}}}
        CurveIndex=&RandomRangeFloat(Minimum=0.3, Maximum=0.6)
        Scale Curve=50
       )
       B=50
      )
     )
    )) {isSetParameters=true, module=SetVariables_#1}
   &AddVelocity(
    Module.Use Falloff Distance=false
    Override Seed=false
    Velocity Mode=::NewEnumerator1
    Velocity Origin=$Particles.Position
    Velocity Speed=40
   )
   &AddVelocity(
    Cone Angle=50
    Inner Cone Angle=25
    Module.UseConeFalloff=false
    Module.UseVelDistribution=false
    Override Seed=false
    Velocity Mode=::NewEnumerator2
    Velocity Speed=&RandomRangeFloat(Minimum=275, Maximum=375)
   ) {module=AddVelocity001}
  ]
  particleUpdateScript [
   &ParticleState()
   &Drag(Drag=&RandomRangeFloat(Minimum=2, Maximum=4), Use Linear Drag=true, Use Rotational Drag=true)
   &CurlNoiseForce(Noise Frequency=5, Noise Strength=50, Pan Noise=false)
   &GravityForce(Gravity=&RandomRangeVector(Minimum=<0 0 50>, Maximum=<0 0 -150>))
   &SolveForcesAndVelocity(Clamp Velocity=false, Limit Acceleration=false)
   &ScaleMeshSize(Scale Factor=&VectorFromFloat(Value=&FloatFromCurve(FloatCurve=DataInterface() {propertyValues {Curve {keys [
          {arriveTangent=9.0725, interpMode=RCIM_Cubic, leaveTangent=9.0725, tangentMode=RCTM_User, time=0, value=0}
          {arriveTangent=0.3207, interpMode=RCIM_Cubic, leaveTangent=0.3207, tangentMode=RCTM_User, time=0.1408, value=1}
          {arriveTangent=-3.5468, interpMode=RCIM_Cubic, leaveTangent=-3.5468, tangentMode=RCTM_User, time=1, value=0}
         ]}}})))
   &UpdateMeshOrientation()
  ]
  rendererClasses [&NiagaraMeshRendererProperties]
  renderers [
   {rendererClass=&NiagaraMeshRendererProperties, rendererIndex=0}
   {rendererClass=&NiagaraMeshRendererProperties, rendererIndex=1}
  ]
 }
 Leaves_Vortex {
  emitterSpawnScript [
   &NM_ToonShaderParameters(
    BaseColor:LinearColor=<rgba 0.1117 0.3646 0 1>
    HighLight_Size=0.15
    HighLight_Softness=1
    InkLine_Color:LinearColor=<rgba 0.2025 0.3958 0.1533 1>
    Metallic_Reflection_Color:LinearColor=<rgba 0.0097 0.5573 0 1>
    RimLight_Color:LinearColor=<rgba 0 0.9427 0.0941 1>
   )
   &NM_OutlineParameters(OutlineColor:LinearColor=<rgba 0.0162 0.0417 0.0042 1>, OutlineWidth=0.75)
  ]
  emitterUpdateScript [
   &EmitterState(Life Cycle Mode=::NewEnumerator1, Loop Behavior=::NewEnumerator1, Loop Duration=0.2, UseLoopDelay=true)
   &SpawnRate(SpawnRate=40, Use Spawn Probability=false)
  ]
  particleSpawnScript [
   &InitializeParticle(
    Color=&FNTimeOfDayColor(
     Night Color=<rgba 0.117 0.2664 0.1441 1>
     Morning Color=<rgba 0.3729 0.5052 0.1395 1>
     Day Color=<rgba 0.3215 0.4271 0.1245 1>
     Evening Color=<rgba 0.3646 0.4844 0.1412 1>
    )
    Lifetime Max=0.5
    Lifetime Min=0.4
    Lifetime Mode=::NewEnumerator1
    Mesh Scale Mode=::NewEnumerator1
    Mesh Uniform Scale Max=1.1
    Mesh Uniform Scale Min=0.8
    Sprite Size Mode=::NewEnumerator3
    Uniform Sprite Size=20
    UsePositionOffset=true
    Write Material Random=true
   )
   &ShapeLocation(
    Disc Coverage=1
    MaskSpawnGroup=false
    Override Seed=false
    Ring Radius=40
    Rotation Mode=::NewEnumerator1
    Shape Primitive=::NewEnumerator4
    UseOwnerScale=true
   )
   &InitialMeshOrientation(Use Rotation Vector=false)
  ]
  particleUpdateScript [
   &ParticleState()
   /SetVariables(Particles.Position=&AddVectorToPosition(
     Position=$Particles.Initial.Position
     Vector=&MakeVector(
      X=&Lerp_Float(A=-225, B=200, Alpha=$Particles.NormalizedAge)
      Y=0
      Z=&Add_Float(
       A=&FloatFromCurve(
        FloatCurve=DataInterface() {propertyValues {Curve {keys [
            {arriveTangent=-2.1826, interpMode=RCIM_Cubic, leaveTangent=-2.1826, tangentMode=RCTM_User, time=0, value=1.5}
            {arriveTangent=-0.5, interpMode=RCIM_Cubic, leaveTangent=-0.5, time=0.5, value=0.5}
            {arriveTangent=2.8781, interpMode=RCIM_Cubic, leaveTangent=2.8781, tangentMode=RCTM_User, time=1, value=1}
           ]}}}
        Scale Curve=50
       )
       B=50
      )
     )
    )) {isSetParameters=true, module=SetVariables_#1}
   &ScaleMeshSize(Scale Factor=&VectorFromFloat(Value=&FloatFromCurve(FloatCurve=DataInterface() {propertyValues {Curve {keys [
          {arriveTangent=9.0725, interpMode=RCIM_Cubic, leaveTangent=9.0725, tangentMode=RCTM_User, time=0, value=0}
          {arriveTangent=0.3207, interpMode=RCIM_Cubic, leaveTangent=0.3207, tangentMode=RCTM_User, time=0.1408, value=1}
          {arriveTangent=-3.5468, interpMode=RCIM_Cubic, leaveTangent=-3.5468, tangentMode=RCTM_User, time=1, value=0}
         ]}}})))
   &self:Postion_Velocity() {module=Postion_Velocity}
   &UpdateMeshOrientation(Rotation Rate=8) {module=UpdateMeshOrientation001}
   &UpdateMeshOrientation(
    Facing Interpolation=false
    Facing Position:Vector3f=$Particles.Velocity
    Orientation Method=::NewEnumerator2
    Override Quat=false
    TransformFacingOrigin=false
    TransformFacingPosition=false
    TransformUp=false
   )
  ]
  rendererClasses [&NiagaraMeshRendererProperties]
  renderers [
   {rendererClass=&NiagaraMeshRendererProperties, rendererIndex=0}
   {rendererClass=&NiagaraMeshRendererProperties, rendererIndex=1}
  ]
 }
 Main_Ring {
  emitterUpdateScript [
   &EmitterState()
   &SpawnParticlesFromOtherEmitter(
    Attribute Reader=DataInterface() {propertyValues {EmitterBinding {bindingMode=Other, emitterName=Source}}}
    Maximum Age Spawn Threshold=0.3
    Spawn Rate=40
    Use Maximum Age=true
    Use Minimum Age=false
   )
  ]
  particleSpawnScript [
   /SetVariables(Particles.RandomRange:NiagaraFloat=&Multiply_Float(
     A=&RandomRangeFloat(Minimum=0.25, Maximum=0.375)
     B=&FloatFromCurve(
      FloatCurve=DataInterface() {propertyValues {Curve {keys [<at 0 1 RCIM_Cubic>, {arriveTangent=7.3942, interpMode=RCIM_Cubic, leaveTangent=7.3942, tangentMode=RCTM_User, time=0.3, value=2}]}}}
      CurveIndex=$Emitter.Age
     )
    )) {isSetParameters=true}
   &InitializeParticle(
    Color=&FNTimeOfDayColor(
     Night Color=<rgba 0.4667 0.5464 0.5833 0.4>
     Morning Color=<rgba 0.5958 0.6976 0.7448 0.6>
     Day Color=<rgba 0.8 0.9367 1 0.6>
     Evening Color=<rgba 0.575 0.6732 0.7188 0.6>
    )
    Lifetime=0.5
    Mesh Scale=&MakeVector(X=$Particles.RandomRange, Y=&RandomRangeFloat(Minimum=0.2, Maximum=0.5), Z=$Particles.RandomRange)
    Mesh Scale Mode=::NewEnumerator4
    UsePositionOffset=false
    Write Material Random=true
   )
   &SampleParticlesFromOtherEmitter(Mesh Orientation Sampling:ENiagara_DirectReadSamplingMode=::NewEnumerator1, Velocity Sampling=::NewEnumerator0)
   &CameraOffset(Camera Offset Amount=15)
   &InitialMeshOrientation(
    Mesh Axis To Orient=<0 1 0>
    Mesh Orientation Mode=::NewEnumerator2
    Rotation=&RandomRangeVector(Minimum=<0 -1 -0.07>, Maximum=<0 1 0.07>)
    Use Rotation Vector=true
   )
   /SetVariables(Particles.Quat:Quat4f=$Output.SampleParticlesFromOtherEmitter.SampledMeshOrientation) {isSetParameters=true}
  ]
  particleUpdateScript [
   &ParticleState()
   /SetVariables(Particles.Position=&AddVectorToPosition(
     Position=$Particles.Initial.Position
     Vector=&RotateVectorByQuaternion(
      Vector To Rotate=&MakeVector(
       X=&Multiply_Float(A=&RandomRangeFloat(Minimum=-25, Maximum=-50), B=$Particles.NormalizedAge)
       Y=0
       Z=0
      )
      Quaternion=$Particles.Quat
     )
    )) {isSetParameters=true}
   &ScaleMeshSize(Scale Factor=&VectorFromFloat(Value=&FloatFromCurve(FloatCurve=DataInterface() {propertyValues {Curve {keys [<at 0 0.7>,<at 1 1>]}}})))
   &DynamicMaterialParameters(
    Index 0 Param 1=&FloatFromCurve(FloatCurve=DataInterface() {propertyValues {Curve {keys [
         {arriveTangent=2.4535, interpMode=RCIM_Cubic, leaveTangent=2.4535, tangentMode=RCTM_User, time=0, value=0}
         {interpMode=RCIM_Cubic, tangentMode=RCTM_User, time=0.5, value=0.5}
         {arriveTangent=-2.5865, tangentMode=RCTM_User, time=1, value=0}
        ]}}})
    Module.Index0Param0Write=true
    Module.Index0Param1Write=false
    Module.Index0Param2Write=false
    Module.Index0Param3Write=false
   )
   &ScaleColor(
    Scale Alpha=&FloatFromCurve(
     FloatCurve=DataInterface() {propertyValues {Curve {keys [
         {arriveTangent=6.3511, interpMode=RCIM_Cubic, leaveTangent=6.3511, tangentMode=RCTM_User, time=0, value=0}
         {interpMode=RCIM_Cubic, tangentMode=RCTM_User, time=0.5, value=1}
         {tangentMode=RCTM_User, time=1, value=0}
        ]}}}
     Scale Curve=&FloatFromCurve(
      FloatCurve=DataInterface() {propertyValues {Curve {keys [
          {interpMode=RCIM_Cubic, tangentMode=RCTM_User, time=0, value=0}
          {arriveTangent=9.6396, interpMode=RCIM_Cubic, leaveTangent=0.1095, tangentMode=RCTM_Break, time=0.3, value=1}
          {arriveTangent=-3.0223, interpMode=RCIM_Cubic, leaveTangent=-3.0223, tangentMode=RCTM_User, time=0.7, value=0}
         ]}}}
      CurveIndex=$Emitter.Age
      Scale Curve=0.5
     )
    )
    Scale RGB=<0.5 0.5 0.5>
    ScaleA=true
    ScaleRGB=true
   )
   &UpdateMeshOrientation(Rotation Rate=2, Rotation Vector=<0 1 0>) {module=UpdateMeshOrientation001}
   &UpdateMeshOrientation(
    Facing Interpolation=false
    Facing Position:Vector3f=$Particles.Velocity
    Orientation Method=::NewEnumerator2
    Override Quat=false
    TransformFacingOrigin=false
    TransformFacingPosition=false
    TransformUp=false
   )
   /SetVariables(Particles.New Linear Color:LinearColor=&MakeLinearColorFromVectorAndFloat(
     "Vector (RGB)"=&MakeVectorFromLinearColorRGB(Color=<rgba 0.7 0.8816 1 1>)
     "Float (Alpha)"=&Multiply_Float(
      A=&Multiply_Float(
       A=&FN_FrontEndFloatModifier(Gameplay=0.7, Front End=1.4)
       B=&FNTimeOfDayScalar(Night Scale=0.5, Morning Scale=0.3, Evening Scale=0.3)
      )
      B=&FloatFromCurve(
       FloatCurve=DataInterface() {propertyValues {Curve {keys [
           {interpMode=RCIM_Cubic, tangentMode=RCTM_User, time=0, value=0}
           {interpMode=RCIM_Cubic, tangentMode=RCTM_User, time=0.5, value=1}
           {tangentMode=RCTM_User, time=1, value=0}
          ]}}}
       Scale Curve=&FloatFromCurve(
        FloatCurve=DataInterface() {propertyValues {Curve {keys [
            {interpMode=RCIM_Cubic, tangentMode=RCTM_User, time=0, value=0}
            {arriveTangent=9.6396, interpMode=RCIM_Cubic, leaveTangent=0.1095, tangentMode=RCTM_Break, time=0.3, value=1}
            {arriveTangent=-3.0223, interpMode=RCIM_Cubic, leaveTangent=-3.0223, tangentMode=RCTM_User, time=0.7, value=0}
           ]}}}
        CurveIndex=$Emitter.Age
        Scale Curve=0.6
       )
      )
     )
    )) {isSetParameters=true}
   /SetVariables(Particles.New Vector 4f:Vector4f=&MakeVector4(
     X=&FloatFromCurve(FloatCurve=DataInterface() {propertyValues {Curve {keys [
          {interpMode=RCIM_Cubic, tangentMode=RCTM_User, time=0, value=0}
          {interpMode=RCIM_Cubic, tangentMode=RCTM_User, time=0.5, value=1}
          {tangentMode=RCTM_User, time=1, value=0}
         ]}}})
     Y=0
     Z=0
     W=0
    )) {isSetParameters=true}
   /SetVariables(Particles.BlurMeshScale:Vector3f=&Multiply_Vector(A=$Particles.Initial.Scale, B=<0.85 0.85 0.85>)) {isSetParameters=true}
  ]
  rendererClasses [&NiagaraMeshRendererProperties]
  renderers [
   {rendererClass=&NiagaraMeshRendererProperties, rendererIndex=0}
   {rendererClass=&NiagaraMeshRendererProperties, rendererIndex=1}
  ]
 }
 Puffs {
  emitterUpdateScript [
   &EmitterState(Life Cycle Mode=::NewEnumerator1, Loop Behavior=::NewEnumerator1, Loop Delay=0.2, Loop Duration=0.4, UseLoopDelay=true)
   &SpawnRate(SpawnRate=30, Use Spawn Probability=false)
   &SpawnRate(SpawnGroup=1, SpawnRate=30, Use Spawn Probability=false) {module=SpawnRate001}
  ]
  particleSpawnScript [
   &InitializeParticle(
    Color=<rgba 0.3125 0.2184 0.1172 1>
    Lifetime Max=0.7
    Lifetime Min=0.6
    Lifetime Mode=::NewEnumerator1
    Mesh Scale Mode=::NewEnumerator3
    Mesh Uniform Scale=&Multiply_Float(
     A=&RandomRangeFloat(Minimum=1.25, Maximum=2)
     B=&FloatFromCurve(FloatCurve=DataInterface() {propertyValues {Curve {keys [<at 0 1>,<at 1 0.7>]}}}, CurveIndex=$Emitter.NormalizedLoopAge)
    )
    UsePositionOffset=false
    Write Material Random=true
   )
   &SystemLocation(
    MaskSpawnGroup=false
    Offset=&MakeVector(X=&Lerp_Float(A=-100, B=125, Alpha=$Emitter.NormalizedLoopAge), Y=0, Z=0)
   )
   &DynamicMaterialParameters(
    Index 0 Param 3=&RandomRangeFloat(Minimum=0, Maximum=1)
    Module.Index0Param0Write=true
    Module.Index0Param1Write=true
    Module.Index0Param2Write=true
    Module.Index0Param3Write=true
   )
   &AddVelocity(
    Velocity=&Lerp_Vector(
     A=<-200 200 0>
     B=<-200 -200 0>
     Alpha=&MakeFloatFromInt(Int=&SpawnGroup())
    )
    Velocity Speed Scale=&FloatFromCurve(
     FloatCurve=DataInterface() {propertyValues {Curve {keys [
         {interpMode=RCIM_Cubic, tangentMode=RCTM_User, time=0, value=1}
         {arriveTangent=-0.5471, interpMode=RCIM_Cubic, leaveTangent=-0.5471, tangentMode=RCTM_User, time=1, value=0.8}
        ]}}}
     CurveIndex=$Emitter.NormalizedLoopAge
    )
   )
   &InitialMeshOrientation(Use Rotation Vector=false)
  ]
  particleUpdateScript [
   &ParticleState()
   &ScaleMeshSize(Scale Factor=&VectorFromFloat(Value=&FloatFromCurve(FloatCurve=DataInterface() {propertyValues {Curve {keys [{interpMode=RCIM_Cubic, tangentMode=RCTM_User, time=0, value=0.5}, {interpMode=RCIM_Cubic, leaveTangent=2.6782, time=1, value=1}]}}})))
   &UpdateMeshOrientation(Rotation Rate=0.5, Rotation Vector=&RandomVector(Vector Scale=1))
   &AccelerationForce(Acceleration=&Multiply_VectorByFloat(
     Vector=<0 0 200>
     Float=&FloatFromCurve(
      FloatCurve=DataInterface() {propertyValues {Curve {keys [{interpMode=RCIM_Cubic, tangentMode=RCTM_User, time=0.5, value=1}, {interpMode=RCIM_Cubic, tangentMode=RCTM_User, time=1, value=0}]}}}
      CurveIndex=$Emitter.NormalizedLoopAge
     )
    ))
   &Drag(Drag=4, Use Linear Drag=true, Use Rotational Drag=true)
   &SolveForcesAndVelocity(Clamp Velocity=false, Limit Acceleration=false)
   &ScaleColor(
    Scale Alpha=&FloatFromCurve(FloatCurve=DataInterface() {propertyValues {Curve {keys [
         {interpMode=RCIM_Cubic, tangentMode=RCTM_User, time=0, value=1}
         {arriveTangent=-5.3199, interpMode=RCIM_Cubic, leaveTangent=-5.3199, tangentMode=RCTM_User, time=1, value=-1}
        ]}}})
    ScaleA=true
    ScaleRGB=true
   )
  ]
  rendererClasses [&NiagaraMeshRendererProperties]
  renderers [
   {rendererClass=&NiagaraMeshRendererProperties, rendererIndex=0}
   {rendererClass=&NiagaraMeshRendererProperties, rendererIndex=1}
  ]
 }
 Source {
  emitterUpdateScript [
   &EmitterState()
   &SpawnBurst_Instantaneous(Spawn Time=0.01, Use Spawn Probability=false, UseLoopCountLimit=true)
  ]
  particleSpawnScript [&InitializeParticle(
    Lifetime=0.3
    Mesh Scale Mode=::NewEnumerator3
    Mesh Uniform Scale=0
    Sprite Size Mode=::NewEnumerator3
    Uniform Sprite Size=0
    UsePositionOffset=false
    Write Material Random=true
   )]
  particleUpdateScript [
   &SolveForcesAndVelocity(Clamp Velocity=false, Limit Acceleration=false)
   &ParticleState()
   /SetVariables(Particles.Position=&AddVectorToPosition(
     Position=<0 0 0>
     Vector=&MakeVector(
      X=&Lerp_Float(A=-250, B=200, Alpha=$Particles.NormalizedAge)
      Y=0
      Z=&Add_Float(
       A=&FloatFromCurve(
        FloatCurve=DataInterface() {propertyValues {Curve {keys [
            {arriveTangent=-2.1826, interpMode=RCIM_Cubic, leaveTangent=-2.1826, tangentMode=RCTM_User, time=0, value=1.5}
            {arriveTangent=-0.5, interpMode=RCIM_Cubic, leaveTangent=-0.5, time=0.5, value=0.5}
            {arriveTangent=2.8781, interpMode=RCIM_Cubic, leaveTangent=2.8781, tangentMode=RCTM_User, time=1, value=1}
           ]}}}
        Scale Curve=50
       )
       B=50
      )
     )
    )) {isSetParameters=true, module=SetVariables_#1}
   &self:Postion_Velocity() {module=Postion_Velocity}
   &UpdateMeshOrientation(
    Facing Interpolation=false
    Facing Position:Vector3f=$Particles.Velocity
    Orientation Method=::NewEnumerator2
    Override Quat=false
    TransformFacingOrigin=false
    TransformFacingPosition=false
    TransformUp=false
   )
  ]
  rendererClasses [&NiagaraSpriteRendererProperties]
  renderers [{rendererClass=&NiagaraSpriteRendererProperties}]
 }
 Swoosh {
  emitterUpdateScript [
   &EmitterState(Life Cycle Mode=::NewEnumerator1, Loop Behavior=::NewEnumerator1, Loop Delay=0.2, Loop Duration=0.4, UseLoopDelay=true)
   &SpawnRate(SpawnRate=15, Use Spawn Probability=false)
   &SpawnRate(SpawnGroup=1, SpawnRate=15, Use Spawn Probability=false) {module=SpawnRate001}
  ]
  particleSpawnScript [
   &InitializeParticle(
    Color=<rgba 0.3125 0.2184 0.1172 1>
    Lifetime Max=0.6
    Lifetime Min=0.5
    Lifetime Mode=::NewEnumerator1
    Mesh Scale Mode=::NewEnumerator3
    Mesh Uniform Scale=&Multiply_Float(
     A=&RandomRangeFloat(Minimum=1.2, Maximum=1.4)
     B=&FloatFromCurve(FloatCurve=DataInterface() {propertyValues {Curve {keys [<at 0 1>,<at 1 0.5>]}}}, CurveIndex=$Emitter.NormalizedLoopAge)
    )
    Position Offset Coordinate Space=::Local
    UsePositionOffset=true
    Write Material Random=true
   )
   &SystemLocation(
    MaskSpawnGroup=false
    Offset=&MakeVector(X=&Lerp_Float(A=-75, B=150, Alpha=$Emitter.NormalizedLoopAge), Y=0, Z=0)
   )
   &DynamicMaterialParameters(
    Index 0 Param 3=&RandomRangeFloat(Minimum=0, Maximum=1)
    Module.Index0Param0Write=true
    Module.Index0Param1Write=true
    Module.Index0Param2Write=true
    Module.Index0Param3Write=false
   )
   &InitialMeshOrientation(Mesh Orientation Mode=::NewEnumerator1, Use Rotation Vector=false)
   &AddVelocity(
    Velocity=&Lerp_Vector(
     A=<-200 100 0>
     B=<-200 -100 0>
     Alpha=&MakeFloatFromInt(Int=&SpawnGroup())
    )
    Velocity Speed Scale=&FloatFromCurve(
     FloatCurve=DataInterface() {propertyValues {Curve {keys [
         {interpMode=RCIM_Cubic, tangentMode=RCTM_User, time=0, value=1}
         {arriveTangent=-0.5471, interpMode=RCIM_Cubic, leaveTangent=-0.5471, tangentMode=RCTM_User, time=1, value=0.8}
        ]}}}
     CurveIndex=$Emitter.NormalizedLoopAge
    )
   )
  ]
  particleUpdateScript [
   &ParticleState()
   &Drag(Drag=4, Use Linear Drag=true, Use Rotational Drag=true)
   &SolveForcesAndVelocity(Clamp Velocity=false, Limit Acceleration=false)
   &ScaleMeshSize(Scale Factor=&VectorFromCurve(VectorCurve=DataInterface() {propertyValues {
       XCurve {keys [
         {arriveTangent=2.0371, interpMode=RCIM_Cubic, leaveTangent=2.0371, tangentMode=RCTM_User, time=0, value=0.5}
         {interpMode=RCIM_Cubic, leaveTangent=-0.4736, time=1, value=1.25}
        ]}
       YCurve {keys [{arriveTangent=2.0371, interpMode=RCIM_Cubic, leaveTangent=2.0371, tangentMode=RCTM_User, time=0, value=0.5}, <at 1 1.25 RCIM_Cubic>]}
       ZCurve {keys [
         {arriveTangent=3.1042, interpMode=RCIM_Cubic, leaveTangent=3.1042, tangentMode=RCTM_User, time=0, value=0.2}
         {interpMode=RCIM_Cubic, leaveTangent=1.2857, time=1, value=1.5}
        ]}
      }}))
   &ScaleColor(
    Scale Alpha=&FloatFromCurve(FloatCurve=DataInterface() {propertyValues {Curve {keys [
         {interpMode=RCIM_Cubic, tangentMode=RCTM_User, time=0.25, value=1}
         {arriveTangent=-4.7266, interpMode=RCIM_Cubic, leaveTangent=-4.7266, tangentMode=RCTM_User, time=1, value=-1}
        ]}}})
    ScaleA=true
    ScaleRGB=true
   )
  ]
  rendererClasses [&NiagaraMeshRendererProperties]
  renderers [
   {rendererClass=&NiagaraMeshRendererProperties, rendererIndex=0}
   {rendererClass=&NiagaraMeshRendererProperties, rendererIndex=1}
  ]
 }
 Wind_Splines {
  emitterUpdateScript [
   &EmitterState(Life Cycle Mode=::NewEnumerator1, Loop Behavior=::NewEnumerator1, Loop Duration=0.3, UseLoopDelay=false)
   &SpawnRate(SpawnRate=12, Use Spawn Probability=false)
  ]
  particleSpawnScript [
   &InitializeParticle(
    Color=<rgba 0.6185 0.7775 1 0.5>
    Lifetime Max=1
    Lifetime Min=0.5
    Lifetime Mode=::NewEnumerator1
    Mesh Renderer Array Visibility Mode=::NewEnumerator3
    Mesh Renderer Info=DataInterface() {propertyValues {MeshRenderer=&self:Wind_Splines_7.NiagaraMeshRendererProperties_0}}
    Mesh Scale Mode=::NewEnumerator3
    Sprite Size Mode=::NewEnumerator3
    Uniform Sprite Size=0
    UsePositionOffset=false
    Write Material Random=true
   )
   &ShapeLocation(
    Disc Coverage=1
    MaskSpawnGroup=false
    Offset=&RandomRangeVector(Minimum=<-300 0 85>, Maximum=<-250 0 85>)
    Offset Mode=::NewEnumerator0
    Override Seed=false
    Rotation Mode=::NewEnumerator1
    Shape Primitive=::NewEnumerator4
    UseOwnerScale=true
   )
   &InitialMeshOrientation(
    Mesh Orientation Mode=::NewEnumerator1
    Rotation=&RandomRangeVector(Minimum=<-1 0 0>, Maximum=<1 0 0>)
    Use Rotation Vector=true
   )
   &AddVelocity(Velocity=<100 0 0>, Velocity Speed Scale=&RandomRangeFloat(Minimum=4, Maximum=12))
  ]
  particleUpdateScript [
   &ParticleState()
   &DynamicMaterialParameters(
    Index 0 Param 1=&FloatFromCurve(FloatCurve=DataInterface() {propertyValues {Curve {keys [
         {arriveTangent=0.1847, interpMode=RCIM_Cubic, leaveTangent=0.1847, tangentMode=RCTM_User, time=0, value=0}
         {interpMode=RCIM_Cubic, tangentMode=RCTM_User, time=0.4, value=1}
         {arriveTangent=-1.0154, interpMode=RCIM_Cubic, leaveTangent=-1.0154, tangentMode=RCTM_User, time=1, value=0}
        ]}}})
    Index 0 Param 2=&RandomRangeFloat(Minimum=0.75, Maximum=1.3)
    Index 0 Param 3=&RandomRangeFloat(Minimum=0, Maximum=1)
    Index 0 Param 4=&RandomRangeFloat(Minimum=0, Maximum=1)
    Module.Index0Param0Write=true
    Module.Index0Param1Write=true
    Module.Index0Param2Write=true
    Module.Index0Param3Write=true
   )
   &Drag(Drag=&RandomRangeFloat(Minimum=2, Maximum=3), Use Linear Drag=true, Use Rotational Drag=true)
   &SolveForcesAndVelocity(Clamp Velocity=false, Limit Acceleration=false)
   &UpdateMeshOrientation(Rotation Rate=0.25)
   &ScaleMeshSize(Scale Factor=&MakeVector(
     X=&FloatFromCurve(FloatCurve=DataInterface() {propertyValues {Curve {keys [<at 0 0.5>,<at 1 2>]}}})
     Y=1
     Z=1
    ))
   &ScaleColor(
    Scale Alpha=&FloatFromCurve(FloatCurve=DataInterface() {propertyValues {Curve {keys [
         {arriveTangent=9.0725, interpMode=RCIM_Cubic, leaveTangent=9.0725, tangentMode=RCTM_User, time=0, value=0}
         {arriveTangent=-0.7232, interpMode=RCIM_Cubic, leaveTangent=-0.7232, tangentMode=RCTM_User, time=0.15, value=1}
         {arriveTangent=-0.0405, interpMode=RCIM_Cubic, leaveTangent=-0.0405, tangentMode=RCTM_User, time=1, value=0}
        ]}}})
    ScaleA=true
    ScaleRGB=true
   )
  ]
  rendererClasses [&NiagaraMeshRendererProperties]
  renderers [{rendererClass=&NiagaraMeshRendererProperties, rendererIndex=0}]
 }
}
summary {
 systemName=NS_Emote_RemoteControl
 emitters [
  {emitterName=Main_Ring, rendererClasses [&NiagaraMeshRendererProperties]}
  {emitterName=Dust, rendererClasses [&NiagaraSpriteRendererProperties]}
  {emitterName=Source, rendererClasses [&NiagaraSpriteRendererProperties]}
  {emitterName=Air_Trail, rendererClasses [&NiagaraRibbonRendererProperties]}
  {emitterName=Leaves_Vortex, rendererClasses [&NiagaraMeshRendererProperties]}
  {emitterName=Leaves_Drag, rendererClasses [&NiagaraMeshRendererProperties]}
  {emitterName=Swoosh, rendererClasses [&NiagaraMeshRendererProperties]}
  {emitterName=Puffs, rendererClasses [&NiagaraMeshRendererProperties]}
  {emitterName=Wind_Splines, rendererClasses [&NiagaraMeshRendererProperties]}
 ]
 userVariables [
  User.bIsFrontEnd {type=NiagaraBool}
  User.bIsFrontEndPreview {type=NiagaraBool}
 ]
}
